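; Game state component that spawns a random number of rideable llama AI across the Blastberry map during Warmup.
; On authority, it listens for game phase step updates, then reads curve table rows for the enable flag and spawn count range.
; It loads the spawner data class asynchronously, then requests each spawn at a random ground point within the first safe zone radius around map center.
; A missing AI spawner logs an error.
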(event EventBeginPlay
  (Networking|SwitchHasAuthority
    (:Authority
      (bind (_asynctaskproxy _payload) (GameplayEventRouter|StartListeningToStatefulEvent (GameplayEventRouter|GetGlobalEventRouter) "/Script/FortniteGame.FortGamePhaseStepUpdatedEvent")
        (:OnEventReceived
          _)
        (:OnSavedEventStateExists
          _))
      (if (Utilities|Enum|Equal(Enum) (Utilities|Struct|BreakFortGamePhaseStepUpdatedEvent _payload) "Warmup")
        (bind (_outxy _returnvalue) (Fortnite|EvaluateCurveTableRow (Variables|Default|GetIsRidingLlamaEnabled)))
        (if (and _returnvalue (> _outxy 0))
          (bind _outxy_2 (Fortnite|EvaluateCurveTableRow (Variables|Default|GetMinSpawn)))
          (bind _output_get (Variables|Default|SetMinSpawnValue _outxy_2))
          (bind _outxy_3 (Fortnite|EvaluateCurveTableRow (Variables|Default|GetMaxSpawn)))
          (bind _output_get_2 (Variables|Default|SetMaxSpawnValue _outxy_3))
          (bind _output_get_3 (Variables|Default|SetNumToSpawn (Math|Random|RandomIntegerinRange (Math|Float|Truncate _output_get) (Math|Float|Truncate _output_get_2))))
          (if (> _output_get_3 0)
            (Utilities|LoadRidingAssets (Utilities|Array|MakeArray (Variables|Default|GetAIBotSpawnerData%20Class)) (Utilities|Array|MakeArray) (delegate Custom|CustomEvent))))))))

(event Custom|CustomEvent
  (params (bSuccess :type Boolean) (Loaded :type Array%20of%20Object%20References))
  (bind _result (Math|Integer|DecrementInt (Variables|Default|GetNumToSpawn))
    (:%20%20
      (for _ (range (+ _result 1))
        (CallFunction|SpawnandConfigureAI)))))

(fn SpawnAndConfigureAI
  (bind _returnvalue (AI|GetAISpawner))
  (Utilities|IsValid _returnvalue
    (:Is%20Valid
      (LocalVariables|SetAISpawner _returnvalue)
      (bind _returnvalue_2 (Class|FortAthenaAISpawnerData|CreateComponentListfromClass (Utilities|ResolveSoftReference (Variables|Default|GetAIBotSpawnerData%20Class)) self))
      (bind _outlocation (Helpers|RandomGroundLocationinCircle (Class|FortAthenaMapInfo|GetMapCenter (Variables|/Script/FortniteGame.FortGameStateAthena|GetMapInfo (Fortnite|Athena|GameState|GetGameStateAthena))) (SafeZones|GetSafeZoneRadius (Variables|/Script/FortniteGame.FortGameStateAthena|GetMapInfo (Fortnite|Athena|GameState|GetGameStateAthena)) 0) (Variables|/Script/FortniteGame.FortGameStateAthena|GetMapInfo (Fortnite|Athena|GameState|GetGameStateAthena)) 2.5e+04 -2.5e+04 1e+03 "(Name=\"FortSupplyDrop\")"))
      (Class|AthenaAISpawner|RequestSpawn (LocalVariables|GetAISpawner) _returnvalue_2 (Math|Transform|MakeTransform _outlocation) false -1))
    (:Is%20Not%20Valid
      (Development|PrintString "AI Spawn Error: We can't find the AISpawner to spawn NPC , please make sure that \"Bots Enabled\" is checked in \"Editor Preferences > Battle Royale Cheats > Bots\"." true true "(R=1.000000,G=0.000000,B=0.000000,A=1.000000)" 1e+01)
      (Fortnite|FortShippingLog self "AI Spawn Error: We can't find the AISpawner to spawn NPC , please make sure that \"Bots Enabled\" is checked in \"Editor Preferences > Battle Royale Cheats > Bots\"."))))
