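; Blueprint interface for Labrador riding that defines functions a rideable or mounted actor implements.
; ExecuteMountedCue takes gameplay cue parameters and a notify-only flag, so callers can trigger mount-related gameplay cues.
; GetJumpInfo returns two float values and a boolean describing jump state.
; The file holds only signatures with default returns and no logic.
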(fn ExecuteMountedCue
  (params (Cue%20Parameters :type Gameplay%20Cue%20Parameters%20Structure) (Only%20Notifies :type Boolean)))

(fn GetJumpInfo
  (return 0.0 0.0 false))
