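; Ridable component for the Labrador wildlife pawn, managing mounting, dismounting and training, and ejecting riders when it topples.
; On begin play, authority grants the training ability set if enabled.
; Mounting hides the map marker, grants a tracking item until fully trained, and adds a camera modifier toggled by looping targeting-tag waits.
; Dismounting removes both, stops montages, and unconverts, follows or re-marks the pawn based on training and friendly-fire settings.
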(event EventBeginPlay
  (|CallParentFunction)
  (bind _asbp_aipawn_labrador_ridable (Utilities|Casting|CastToBP_AIPawn_Labrador_Ridable (Variables|Default|GetNPCPawnWildlifeRef)))
  (bind _output_get (Variables|Default|SetLabradorPawn _asbp_aipawn_labrador_ridable))
  (Default|BindEventtoToppledDispatcher _output_get (EventDispatchers|CreateEvent :!event EventToppled))
  (Utilities|FlowControl|Sequence
    (:then_1
      (if (ScalableFloat|GetValueasBool (Variables|Default|GetHF_TrainingEnabled))
        (Default|BindEventtoTrainingIncreased (Variables|Default|GetLabradorPawn) (EventDispatchers|CreateEvent :!event LabradorRiding_TrainingIncreased))
        (Networking|SwitchHasAuthority
          (:Authority
            (bind _returnvalue (Ability|EquipFortAbilitySet (Variables|Default|GetLabradorPawn) "/LabradorRidingContent/Gameplay/Training/AS_Labrador_Training.AS_Labrador_Training"))
            (Variables|Default|SetTrainingAbilitySet _returnvalue)
            (Abilities|TryActivateAbilitybyClass (Variables|/Script/FortniteGame.FortAIPawn|GetAIPawnAbilitySystemComponent (Variables|Default|GetLabradorPawn)) "/LabradorRidingContent/Gameplay/Training/GA_Labrador_TrainingManager.GA_Labrador_TrainingManager_C" true)))))))

(event EventHandleRiderStartedRiding
  (params (Rider :type Rider%20Component%20Object%20Reference) (SlotIndex :type Integer))
  (bind _returnvalue (Components|GetOwner Rider))
  ; Unconvert when player dismounts
  (|CallParentFunction Rider 0)
  (EventDispatchers|BindEventtoOnStoppedRidingSingularDeprecated Rider (EventDispatchers|CreateEvent :!event UnconvertLabrador))
  (Components|Activation|Deactivate (Variables|/Labrador/Pawn/BP_AIPawn_Labrador_Ridable.BP_AIPawn_Labrador_Ridable_C|GetNPCSpecialActorComponent (Variables|Default|GetLabradorPawn))) ; Remove map mark while riding
  (Utilities|FlowControl|Sequence
    (:then_0
      (if (ScalableFloat|GetValueasBool (Variables|Default|GetHF_TrackingEnabled))
        (CallFunction|TryApplyTracking (Components|GetOwner Rider))))
    (:then_1
      (bind _returnvalue_2 (Fortnite|Camera|AddCameraModifierforOwningPlayerController "/LabradorRidingContent/Gameplay/Camera/LabradorRiding_CameraModifier_RidingCamera_Labrador.LabradorRiding_CameraModifier_RidingCamera_Labrador_C" _returnvalue))
      (bind _output_get (Variables|Default|SetTargetingCamMod _returnvalue_2))
      (CameraModifier|DisableModifier _output_get true)
      (CallFunction|WaitforTargetingStart _returnvalue))))

(event EventHandleRiderStoppedRiding
  (params (Rider :type Rider%20Component%20Object%20Reference) (SlotIndex :type Integer))
  ; Unconvert when player dismounts
  (|CallParentFunction Rider 0)
  (EventDispatchers|UnbindEventfromOnStoppedRidingSingularDeprecated Rider (EventDispatchers|CreateEvent :!event UnconvertLabrador)))

(event Custom|WaitForTargetingStart
  (params (TargetActor :type Actor%20Object%20Reference))
  ; Unconvert when player dismounts
  (bind _asynctaskproxy (Ability|Tasks|WaitGameplayTagAddToActor TargetActor "(TagName=\"Gameplay.Status.IsTargeting\")" true)
    (:then
      (Variables|Default|SetTask_Targeting _asynctaskproxy))
    (:Added
      (bind _donotuse_primaryrider (Variables|RidableComponent|GetDoNotUse_PrimaryRider)
        (:then
          (CameraModifier|EnableModifier (Variables|Default|GetTargetingCamMod))
          (CallFunction|WaitforTargetingEnd (Components|GetOwner _donotuse_primaryrider)))
        (:else)))))

(event Custom|WaitForTargetingEnd
  (params (TargetActor :type Actor%20Object%20Reference))
  ; Unconvert when player dismounts
  (bind _asynctaskproxy (Ability|Tasks|WaitGameplayTagRemoveFromActor TargetActor "(TagName=\"Gameplay.Status.IsTargeting\")" true)
    (:then
      (Variables|Default|SetTask_EndTargeting _asynctaskproxy))
    (:Removed
      (bind _donotuse_primaryrider (Variables|RidableComponent|GetDoNotUse_PrimaryRider)
        (:then
          (CameraModifier|DisableModifier (Variables|Default|GetTargetingCamMod) true)
          (CallFunction|WaitforTargetingStart (Components|GetOwner _donotuse_primaryrider)))
        (:else)))))

(event Custom|Server_TriggerUnconversion
  (params (Target :type Actor%20Component%20Object%20Reference))
  ; Unconvert when player dismounts
  (bind _asfort_pawn (Utilities|Casting|CastToFortPawn (Components|GetOwner Target)))
  (Utilities|FlowControl|Sequence
    (:then_0
      (if (ScalableFloat|GetValueasBool (Variables|Default|GetHF_TrackingEnabled))
        (Inventory|RemoveItemfromInventoryOwner (Fortnite|GetFortPlayerControllerfromActor _asfort_pawn) "/Labrador/Gameplay/Abilities/Tracking/PAID_LabradorRiding.PAID_LabradorRiding")))
    (:then_1
      (if (ScalableFloat|GetValueasBool (Variables|Default|GetHF_FriendlyFireEnabled))
        (Convert|TriggerUnconversion _asfort_pawn (Variables|Default|GetLabradorPawn) "Dismiss")
        (if (>= (Variables|/Labrador/Pawn/BP_AIPawn_Labrador_Ridable.BP_AIPawn_Labrador_Ridable_C|GetTrainingPercentage (Variables|Default|GetLabradorPawn)) 1.0)
          (AI|Components|Blackboard|SetValueasObject (AI|GetBlackboard (Variables|Default|GetLabradorPawn)) (Utilities|Name|MakeLiteralName "Leader") _asfort_pawn) ; Leader is cleared by default. Add again so that we follow this pawn, but not teammed up
          (else
            (Components|Activation|Activate (Variables|/Labrador/Pawn/BP_AIPawn_Labrador_Ridable.BP_AIPawn_Labrador_Ridable_C|GetNPCSpecialActorComponent (Variables|Default|GetLabradorPawn)) false) ; Reapply map mark
          ))
        (elif (not (>= (Variables|/Labrador/Pawn/BP_AIPawn_Labrador_Ridable.BP_AIPawn_Labrador_Ridable_C|GetTrainingPercentage (Variables|Default|GetLabradorPawn)) 1.0))
          (Convert|TriggerUnconversion _asfort_pawn (Variables|Default|GetLabradorPawn) "Dismiss")
          (Components|Activation|Activate (Variables|/Labrador/Pawn/BP_AIPawn_Labrador_Ridable.BP_AIPawn_Labrador_Ridable_C|GetNPCSpecialActorComponent (Variables|Default|GetLabradorPawn)) false) ; Reapply map mark
        )))
    (:then_2
      (CallFunction|CleanupMontages _asfort_pawn))))

(event Custom|CleanupMontages
  (params (Rider :type Character%20Object%20Reference))
  ; Cleanup montages
  (Animation|Montage|MontageStop (Components|SkeletalMesh|GetAnimInstance (Variables|/Script/Engine.Character|GetMesh Rider)) 0.15 "/LabradorRidingContent/Animation/HardSaddle/Labrador/LabradorRiding_Player_Boost_N_M.LabradorRiding_Player_Boost_N_M") ; Fix bug when labrador dies
  (Animation|Montage|MontageStop (Components|SkeletalMesh|GetAnimInstance (Variables|/Script/Engine.Character|GetMesh (Variables|Default|GetLabradorPawn))) 0.15 "/Labrador/Animation/Labrador_Run_Bounding_Boost_Mounted_M.Labrador_Run_Bounding_Boost_Mounted_M") ; Fix bug where sprint anim persists
)

(fn CanBeRidden
  (params (Rider :type Rider%20Component%20Object%20Reference) (SlotIndex :type Integer))
  (return (and (|CallParentFunction Rider) (not (Variables|/Labrador/Pawn/BP_AIPawn_Labrador.BP_AIPawn_Labrador_C|GetIsToppled (Variables|Default|GetLabradorPawn))))))

(fn CanJumpExit
  (return (|CallParentFunction)))

(fn EventToppled
  (params (isToppled :type Boolean))
  (if isToppled
    (CallFunction|KickRiderOffSelf)))

(fn GetJumpInfo
  (return (Variables|RidingAttach|GetJumpAttach_GroundHeightMin) (Variables|RidingAttach|GetJumpAttach_GroundHeightBuffer) (Variables|RidingAttach|GetUseAltRidingMessage)))

(fn LabradorRiding_TrainingIncreased
  (params (Value :type Float%20%28double-precision%29))
  (bind _donotuse_primaryrider (Variables|RidableComponent|GetDoNotUse_PrimaryRider))
  (bind _returnvalue (Components|GetOwner _donotuse_primaryrider))
  (Utilities|IsValid _returnvalue
    (:Is%20Valid
      (if (>= Value 1.0)
        (if (ScalableFloat|GetValueasBool (Variables|Default|GetHF_TrackingEnabled))
          (Inventory|RemoveItemfromInventoryOwner (Fortnite|GetFortPlayerControllerfromActor _returnvalue) "/Labrador/Gameplay/Abilities/Tracking/PAID_LabradorRiding.PAID_LabradorRiding"))))))

(fn TryApplyTracking
  (params (Object :type Object%20Reference))
  (bind _asfort_pawn (Utilities|Casting|CastToFortPawn Object))
  (if (not (>= (Variables|/Labrador/Pawn/BP_AIPawn_Labrador_Ridable.BP_AIPawn_Labrador_Ridable_C|GetTrainingPercentage (Variables|Default|GetLabradorPawn)) 1.0))
    (Inventory|GiveItemtoInventoryOwner :InventoryOwner (Fortnite|GetFortPlayerControllerfromActor _asfort_pawn) :ItemDefinition "/Labrador/Gameplay/Abilities/Tracking/PAID_LabradorRiding.PAID_LabradorRiding" :NumberToGive 1 :bNotifyPlayer false :ItemLevel -1 :PickupInstigatorHandle -1 :bUseItemPickupAnalyticEvent false :WeaponAmmoOverride -1)))

(fn UnconvertLabrador
  (params (Rider :type Rider%20Component%20Object%20Reference) (Ridable :type Ridable%20Component%20Object%20Reference))
  (bind _returnvalue (Components|GetOwner Rider))
  (bind _returnvalue_2 (Fortnite|GetFortPlayerControllerfromActor _returnvalue))
  (Utilities|IsValid Rider
    (:Is%20Valid
      (CallFunction|ServerTriggerUnconversion Rider)
      (bind _targetingcammod (Variables|Default|GetTargetingCamMod)
        (:then
          (Utilities|IsValid _returnvalue
            (:Is%20Valid
              (Utilities|IsValid _returnvalue_2
                (:Is%20Valid
                  (bind _playercameramanager (Variables|/Script/Engine.PlayerController|GetPlayerCameraManager _returnvalue_2)
                    (:then
                      (Game|Player|RemoveCameraModifier _playercameramanager _targetingcammod))
                    (:else)))))))
        (:else)))))
