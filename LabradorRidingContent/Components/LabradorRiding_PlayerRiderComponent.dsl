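; Player rider component that lets Fort players jump onto and ride creatures, with UI forwarding for target name, icon, health and energy.
; Ridable hits pass height, descent and tamed-owner checks, then toggle riding on the server, subject to blocking tags.
; On authority, riding applies a GameplayEffect, converts the creature and watches movement changes, tags and events to dismount.
; The owning client binds mount landings for rumble cues; entering water cancels sprint.
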(event EventBeginPlay
  ;[1 1 1] Begin Play
  (bind _asfort_player_pawn (Utilities|Casting|CastToFortPlayerPawn (Components|GetOwner)))
  (Variables|Default|SetFortPlayerRef _asfort_player_pawn)
  (bind _asplayer_pawn_generic_parent (Utilities|Casting|CastToPlayerPawn_Generic_Parent (Components|GetOwner)))
  (Default|AssignOnOverlapWaterVolume _asplayer_pawn_generic_parent (delegate Custom|OnOverlapWaterVolume_Event)))

(event Custom|OnOverlapWaterVolume_Event
  (params (bIsInWater :type Boolean) (Pawn :type Fort%20Player%20Pawn%20Object%20Reference) (Water :type Actor%20Object%20Reference))
  ;[1 1 1] Begin Play
  (if bIsInWater
    (CallFunction|CancelSprint)))

(event EventHandleRidableFound
  (params (HitResult :type Hit%20Result%20Structure%20%28by%20ref%29))
  (bind (_bblockinghit _binitialoverlap _time _distance _location _impactpoint _normal _impactnormal _physmat _hitactor) (Collision|BreakHitResult HitResult))
  ;[1 1 1] Attempt to find creatures to ride
  (bind (_success _targetobject _usealtridingmessage) (CallFunction|TestifValidforRidingAttempt _hitactor))
  (if _success
    (CallFunction|ToggleRidingonServer _targetobject _usealtridingmessage)
    (if (Riding|IsRidingActor)
      (GameplayCue|ExecuteGameplayCue (Variables|Default|GetFortPlayerRef) (Variables|Default|GetJumpOnCreature_Cue_Tag)))))

(event Custom|ToggleRidingOnServer
  (params (TargetObject :type Object%20Reference) (UseAltRidingMessage :type Boolean))
  (bind _fortplayerref (Variables|Default|GetFortPlayerRef))
  ;[1 1 1] Toggle Riding On Server
  (if UseAltRidingMessage
    (Class|LabradorRidingRidingGenericInterafce|ToggleRidingAlternative(Message) TargetObject)
    (else
      (bind _asactor (Utilities|Casting|CastToActor TargetObject))
      (Default|ToggleRiding _fortplayerref _asactor true)
      (Networking|ForceNetUpdate _fortplayerref))))

(event EventHandleStartedRiding
  (params (Ridable :type Ridable%20Component%20Object%20Reference) (SlotIndex :type Integer))
  (bind _donotuse_activeridable (Variables|RiderComponent|GetDoNotUse_ActiveRidable))
  (bind _returnvalue (Components|GetOwner _donotuse_activeridable))
  (bind _returnvalue_2 (Components|GetOwner))
  (bind _returnvalue_3 (Components|GetOwner))
  (bind _returnvalue_4 (Components|GetOwner))
  (bind _donotuse_activeridable_2 (Variables|RiderComponent|GetDoNotUse_ActiveRidable))
  ;[0.6487 1 0.6273] Started Riding
  (|CallParentFunction Ridable 0)
  (Utilities|SwitchHasAuthority
    (:Authority
      (GameplayEffects|FortApplyGameplayEffectToSelf :self (Ability|GetFortAbilitySystemComponent _returnvalue_4) :GameplayEffectClass (Variables|Default|GetGE_PlayerIsRiding) :Level 0.0)
      ;;[1 1 1] Remove Next FallDamage Immunity
      (Variables|/Script/FortniteGame.FortPawn|SetbIgnoreNextFallingDamage false (Variables|Default|GetFortPlayerRef))
      ;;[1 1 1] Tame creature that is ridden (temp)
      (bind _asfort_pawn (Utilities|Casting|CastToFortPawn (Components|GetOwner Ridable))
        (:then
          (Convert|TriggerConversion (Variables|Default|GetFortPlayerRef) _asfort_pawn true))
        (:CastFailed
          _))
      (bind _fortplayerref (Variables|Default|GetFortPlayerRef)
        (:then
          (Character|BindEventtoMovementModeChangedDelegate _fortplayerref (delegate Custom|MovementModeChanged))
          (Utilities|IsValid _returnvalue
            (:Is%20Valid
              (bind _asynctaskproxy (Ability|Tasks|WaitGameplayTagAddToActor _returnvalue "(TagName=\"Riding.Player.BlockRiding\")")
                (:then
                  (Variables|Default|SetActiveRidableBlockRidingGameplayTagAsyncAction _asynctaskproxy)
                  ;;[1 1 1] Listening Gameplay Events to StopRiding
                  (Utilities|FlowControl|DoOnce false
                    (:Completed
                      (Ability|Async|WaitGameplayEventToActor _returnvalue_2 "(TagName=\"Event.Athena.HidingProp.Hide\")"
                        (:then
                          (Ability|Async|WaitGameplayEventToActor _returnvalue_2 "(TagName=\"Abilities.Generic.Vent.Enter\")"
                            (:then
                              ;;[1 1 1] Listening Tags to StopRiding
                              (Ability|Tasks|WaitGameplayTagAddToActor _returnvalue_3 "(TagName=\"Gameplay.Status.InsideTornado\")"
                                (:then
                                  (Ability|Tasks|WaitGameplayTagAddToActor _returnvalue_3 "(TagName=\"Riding.Player.BlockRiding\")"
                                    (:Added
                                      _)))
                                (:Added
                                  _))
                              (Riding|StopRidingActor true))
                            (:EventReceived
                              _)))
                        (:EventReceived
                          _)))))
                (:Added
                  _)))
            (:Is%20Not%20Valid
              _)))
        (:else)))
    (:Remote
      ;[0.2664 0.4072 0.2542] LOCAL CLIENT ONLY = Start listening for creature landed
      (if (Utilities|Enum|Equal(Enum) (Networking|GetLocalRole _returnvalue_4) "ROLE_AutonomousProxy")
        (Utilities|IsValid _donotuse_activeridable_2
          (:Is%20Valid
            (bind _asfort_pawn_2 (Utilities|Casting|CastToFortPawn (Components|GetOwner _donotuse_activeridable_2)))
            (EventDispatchers|BindEventtoOnPawnLanded _asfort_pawn_2 (delegate Custom|MountLanded))))))))

(event EventHandleStoppedRiding
  (params (Ridable :type Ridable%20Component%20Object%20Reference) (SlotIndex :type Integer))
  (bind _fortplayerref (Variables|Default|GetFortPlayerRef))
  ;[1 0.2289 0.2488] Stopped Riding
  (Utilities|FlowControl|Sequence
    (:then_0
      ;;[1 1 1] Stop listening for Movement Mode Change Event
      (Utilities|IsValid _fortplayerref
        (:Is%20Valid
          (Character|UnbindEventfromMovementModeChangedDelegate _fortplayerref (delegate Custom|MovementModeChanged))
          (|CallParentFunction Ridable 0)
          (Utilities|SwitchHasAuthority
            (:Authority
              (GameplayEffects|RemoveActiveEffectswithGrantedTags (Ability|GetFortAbilitySystemComponent (Components|GetOwner)) (Variables|Default|GetT_PlayerIsRiding))
              (bind _activeridableblockridinggameplaytagasyncaction (Variables|Default|GetActiveRidableBlockRidingGameplayTagAsyncAction)
                (:then
                  (AsyncAction|Cancel _activeridableblockridinggameplaytagasyncaction))
                (:else)))
            (:Remote
              ;[0.4072 0.09306 0.1022] CLIENT ONLY = Stop listening for creature landed
              (if (Utilities|Enum|Equal(Enum) (Networking|GetLocalRole (Components|GetOwner)) "ROLE_AutonomousProxy")
                (bind _asfort_pawn (Utilities|Casting|CastToFortPawn (Components|GetOwner Ridable)))
                (EventDispatchers|UnbindEventfromOnPawnLanded _asfort_pawn (delegate Custom|MountLanded))))))))
    (:then_1
      (CallFunction|CleanupMontages (Variables|Default|GetFortPlayerRef)))))

(event Custom|CleanupMontages
  (params (Rider :type Character%20Object%20Reference))
  (Utilities|IsValid Rider
    (:Is%20Valid
      (Animation|Montage|MontageStop (Components|SkeletalMesh|GetAnimInstance (Variables|/Script/Engine.Character|GetMesh Rider)) 0.15 "/LabradorRidingContent/Animation/HardSaddle/Labrador/LabradorRiding_Player_Boost_N_M.LabradorRiding_Player_Boost_N_M") ; Fix bug when labrador dies
    )))

(event Custom|MovementModeChanged
  (params (Character :type Character%20Object%20Reference) (PrevMovementMode :type EMovementMode%20Enum :default "MOVE_None") (PreviousCustomMode :type Byte))
  ;[1 1 1] Movement mode changed
  (bind _returnvalue (GameplayTags|HasAnyMatchingGameplayTags(Message) Character (Variables|Default|GetAllowRidingMovementModeChangeTags)))
  (if (not _returnvalue)
    (Riding|StopRidingActor true)
    (Networking|ForceNetUpdate (Variables|Default|GetFortPlayerRef))))

(event Custom|RiderUIStartClient
  (params (Name :type Text) (Icon :type Slate%20Brush%20Structure) (Ridable :type Actor%20Object%20Reference))
  ;[1 1 1] Riding Target UI Related
  (bind _output_get (Variables|Default|SetRidableActorName Name))
  (bind _output_get_2 (Variables|Default|SetRidableActorIcon Icon))
  (bind _output_get_3 (Variables|Default|SetCachedRidableActor Ridable))
  (Default|CallRiderUINewStats _output_get _output_get_2 _output_get_3))

(event Custom|MountLanded
  (params (Hit :type Hit%20Result%20Structure%20%28by%20ref%29))
  ;[0.3412 0.619 1] Mount Landed - Gameplay Cue for Controller Rumble
  (bind _aslabrador_ridable_interface (Utilities|Casting|CastToLabradorRidableInterface (Variables|RiderComponent|GetDoNotUse_ActiveRidable)))
  (Class|LabradorRidableInterface|ExecuteMountedCue :self _aslabrador_ridable_interface :Only%20Notifies true))

(event Custom|RiderUIEndClient
  (params (Ridable :type Actor%20Object%20Reference))
  ;[1 1 1] Riding Target UI Related
  (Default|CallRiderUIEnd Ridable))

(event Custom|SetTargetHealth
  (params (targetHealth :type Float%20%28double-precision%29))
  ;[1 1 1] Riding Target UI Related
  (Variables|Default|SettargetHealth targetHealth))

(fn GetJumpInfo
  (params (TargetActor :type Actor%20Object%20Reference))
  (bind _returnvalue (Class|LabradorRidingFunctionLibrary|GetComponentbySoftClass TargetActor "/LabradorRidingContent/Components/LabradorRiding_CreatureBaseRidableComponent.LabradorRiding_CreatureBaseRidableComponent_C"))
  (bind _aslabrador_ridable_interface (Utilities|Casting|CastToLabradorRidableInterface _returnvalue)
    (:then
      (bind (_jump_attach_ground_height_min _jump_attach_ground_height_buffer _use_alt_riding_message) (Class|LabradorRidableInterface|GetJumpInfo _aslabrador_ridable_interface))
      (return _returnvalue _jump_attach_ground_height_min _jump_attach_ground_height_buffer _use_alt_riding_message))
    (:CastFailed
      (bind _returnvalue_2 (Class|LabradorRidingFunctionLibrary|GetComponentbySoftClass 0 "/LabradorRidingContent/Components/LabradorRiding_CreatureBaseNonRidableComponent.LabradorRiding_CreatureBaseNonRidableComponent_C"))
      (bind _aslabrador_ridable_interface_2 (Utilities|Casting|CastToLabradorRidableInterface _returnvalue_2)
        (:then
          (bind (_jump_attach_ground_height_min_2 _jump_attach_ground_height_buffer_2 _use_alt_riding_message_2) (Class|LabradorRidableInterface|GetJumpInfo _aslabrador_ridable_interface_2))
          (return 0 _jump_attach_ground_height_min_2 _jump_attach_ground_height_buffer_2 _use_alt_riding_message_2))
        (:CastFailed
          (return 0 0.0 0.0 false))))))

(fn GetRidableActorData
  (bind _cachedridableactor (Variables|Default|GetCachedRidableActor))
  (Utilities|IsValid _cachedridableactor
    (:Is%20Valid
      (return true (Variables|Default|GetRidableActorName) (Variables|Default|GetRidableActorIcon) _cachedridableactor))
    (:Is%20Not%20Valid
      (return false "" 0 0))))

(fn GetRidingInfoFromTarget
  (return 0.0 0.0 false))

(fn OnRep_targetEnergy
  (Default|CallRiderUIUpdateEnergy (Variables|Default|GettargetEnergy)))

(fn OnRep_targetHealth
  (Default|CallRiderUIUpdateHealth (Variables|Default|GettargetHealth)))

(fn TestIfValidForRidingAttempt
  (params (Object :type Object%20Reference))
  (bind _fortplayerref (Variables|Default|GetFortPlayerRef))
  (bind (_returnvalue_x _returnvalue_y _returnvalue_z) (Math|Vector|BreakVector (Transformation|GetActorLocation _fortplayerref)))
  (bind (_velocity_x _velocity_y _velocity_z) (Math|Vector|BreakVector (Variables|/Script/Engine.MovementComponent|GetVelocity_X (Variables|/Script/Engine.Character|GetCharacterMovement _fortplayerref))))
  (Utilities|IsValid Object
    (:Is%20Valid
      (bind _asfort_pawn (Utilities|Casting|CastToFortPawn Object)
        (:then
          ;[1 1 1] Disable stealing mounts - only let current owner ride already tamed pets
          (Utilities|IsValid (Actor|GetComponentbyClass _asfort_pawn "/Script/FortniteGame.FortPawnComponent_Convert")
            (:Is%20Valid
              (if (Convert|GetOnlyLeaderInteraction (Actor|GetComponentbyClass _asfort_pawn "/Script/FortniteGame.FortPawnComponent_Convert"))
                (if (not (or (not (Utilities|IsValid (Convert|GetConverterPawn (Actor|GetComponentbyClass _asfort_pawn "/Script/FortniteGame.FortPawnComponent_Convert")) :!class /Script/Engine.KismetSystemLibrary)) (== (Convert|GetConverterPawn (Actor|GetComponentbyClass _asfort_pawn "/Script/FortniteGame.FortPawnComponent_Convert")) (Variables|Default|GetFortPlayerRef))))
                  (return false 0 false))
                (else
                  _))
              (bind (_targetactorcomponent _jumpattach_groundheightmin _jumpattach_groundheightbuffer _usealtridingmessage) (CallFunction|GetJumpInfo _asfort_pawn))
              (Utilities|IsValid _targetactorcomponent
                (:Is%20Valid
                  (Utilities|IsValid _fortplayerref
                    (:Is%20Valid
                      (bind (_returnvalue_x_2 _returnvalue_y_2 _returnvalue_z_2) (Math|Vector|BreakVector (Transformation|GetActorLocation _asfort_pawn)))
                      (return (Math|Boolean|ANDBoolean (>= (- (- (- _returnvalue_z (Components|Capsule|GetScaledCapsuleHalfHeight (Variables|/Script/Engine.Character|GetCapsuleComponent _fortplayerref))) (- _returnvalue_z_2 (Components|Capsule|GetScaledCapsuleHalfHeight (Variables|/Script/Engine.Character|GetCapsuleComponent _asfort_pawn)))) _jumpattach_groundheightmin) 0) true (<= _velocity_z 0)) (select _usealtridingmessage _targetactorcomponent _asfort_pawn) _usealtridingmessage))
                    (:Is%20Not%20Valid
                      _)))
                (:Is%20Not%20Valid
                  _))
              (return false 0 false))
            (:Is%20Not%20Valid
              _)))
        (:CastFailed
          _)))
    (:Is%20Not%20Valid
      _)))

(fn ToggleRiding
  (params (PlayerPawn :type Actor%20Object%20Reference) (Target :type Actor%20Object%20Reference) (ShouldRide :type Boolean))
  (bind _returnvalue (Actor|GetComponentbyClass PlayerPawn "/Script/RidingCodeRuntime.RiderComponent"))
  (bind _returnvalue_2 (GameplayTags|HasAnyMatchingGameplayTags(Message) Target (Variables|Default|GetT_CannotRideBucket))) ; No riding if held (like a chicken) or energy is depleted
  (if (not _returnvalue_2)
    (bind _returnvalue_3 (GameplayTags|HasAnyMatchingGameplayTags(Message) PlayerPawn (Variables|Default|GetCannot%20Ride%20Rider%20Bucket)))
    (if (not _returnvalue_3)
      (Utilities|IsValid _returnvalue
        (:Is%20Valid
          (bind _ascontrolling_rider_component (Utilities|Casting|CastToControllingRiderComponent _returnvalue)
            (:then
              (Variables|/Script/RidingCodeRuntime.ControllingRiderComponent|SetbCanControlRidable true _ascontrolling_rider_component))
            (:CastFailed
              _))
          (if ShouldRide
            (bind _returnvalue_4 (Riding|StartRidingActor _returnvalue Target 0))
            (return _returnvalue_4)
            (else
              _))))
      (else
        _))
    (bind _returnvalue_5 (Riding|StopRidingActor _returnvalue true true false true))
    (return _returnvalue_5)
    (else
      _)))
