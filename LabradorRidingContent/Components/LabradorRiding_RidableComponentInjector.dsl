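; Blueprint injector component from LabradorRidingContent that adds a configurable ridable component class to its owning actor at runtime.
; On BeginPlay it passes the soft class reference to LoadRidingAssets with a completion delegate.
; When the callback fires, it resolves the reference and adds that component to the owner by class, without checking the success flag.
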(event EventBeginPlay
  (Utilities|LoadRidingAssets (Utilities|Array|MakeArray (Variables|Default|GetComponentClassToInject)) (Utilities|Array|MakeArray) (delegate Custom|On%20Class%20)))

(event Custom|On%20Class%20
  (params (bSuccess :type Boolean) (Loaded :type Array%20of%20Object%20References))
  (Game|AddComponentbyClass (Components|GetOwner) (Utilities|ResolveSoftReference (Variables|Default|GetComponentClassToInject))))
